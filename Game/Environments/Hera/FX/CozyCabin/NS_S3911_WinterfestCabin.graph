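# Niagara system for the Winterfest cozy cabin: persistent, warm glowing floating meshes with halo sprites, despite the emitter name Foreground_Fog001.
# A burst scatters them in a volume whose height arches across X via a curve.
# Waveforms drive bobbing around spawn points and flickering size and material parameters, while a local AvoidPlayer module adds a push offset away from players.
# GPU Snow and CPU Snow_Low_End curl-noise snowfall emitters are disabled.
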NiagaraSystem /Game/Environments/Hera/FX/CozyCabin/NS_S3911_WinterfestCabin
dependencies {
 usedDataInterfaces [&NiagaraDataInterfaceMeshRendererInfo, &NiagaraDataInterfaceCurve]
 usedDynamicInputs []
 usedModules [&SystemState]
}
emitters {
 Foreground_Fog001 {
  emitterUpdateScript [
   &EmitterState(Life Cycle Mode=::NewEnumerator1, Loop Behavior=::NewEnumerator1, Loop Duration Mode=::NewEnumerator1, UseLoopDelay=false)
   &SpawnBurst_Instantaneous(Spawn Count=50, Use Spawn Probability=false, UseLoopCountLimit=false)
  ]
  particleSpawnScript [
   &InitializeParticle(
    Color=<rgba 25 8.6153 3.342 1>
    Lifetime Max=10
    Lifetime Min=7
    Lifetime Mode=::NewEnumerator1
    Mesh Renderer Array Visibility Mode=::NewEnumerator3
    Mesh Renderer Info=DataInterface() {propertyValues {MeshRenderer=&self:Foreground_Fog001_0.NiagaraMeshRendererProperties_0}}
    Mesh Renderer Random Seed:NiagaraInt32=&RandomRangeInt(Minimum=0, Maximum=2)
    Mesh Scale Mode=::NewEnumerator3
    Mesh Uniform Scale=1.7
    Sprite Rotation Mode=::NewEnumerator2
    Sprite Size Mode=::NewEnumerator3
    Sprite UV Mode=::NewEnumerator3
    Uniform Sprite Size=20
    UsePositionOffset=false
    Write Material Random=true
   )
   /SetVariables(Particles.HorizontalDistribution:NiagaraFloat=&RandomRangeFloat(Minimum=0, Maximum=1)) {isSetParameters=true}
   /SetVariables(Particles.Position=&AddVectorToPosition(
     Position=/Niagara/DynamicInputs/Helpers/SimulationPosition()
     Vector=&MakeVector(
      X=&Lerp_Float(A=-450, B=450, Alpha=$Particles.HorizontalDistribution)
      Y=&RandomRangeFloat(Minimum=-600, Maximum=650)
      Z=&Lerp_Float(
       A=&RandomRangeFloat(Minimum=-40, Maximum=-50)
       B=&RandomRangeFloat(Minimum=70, Maximum=125)
       Alpha=&FloatFromCurve(
        FloatCurve=DataInterface() {propertyValues {Curve {keys [
            <at 0 0>
            {arriveTangent=4.2262, interpMode=RCIM_Cubic, leaveTangent=3.5357, tangentMode=RCTM_Break, time=0.4, value=1}
            {interpMode=RCIM_Cubic, tangentMode=RCTM_User, time=0.5, value=1.25}
            {arriveTangent=-3.5052, tangentMode=RCTM_User, time=0.6, value=1}
            <at 1 0>
           ]}}}
        CurveIndex=$Particles.HorizontalDistribution
       )
      )
     )
    )) {isSetParameters=true}
   &ShapeLocation(
    Box Midpoint=&RandomRangeVector(Minimum=<0 0 0>, Maximum=<1 1 1>)
    Box Size=<400 650 50>
    MaskSpawnGroup=false
    Offset Mode=::NewEnumerator0
    Override Seed=false
    Shape Primitive=::NewEnumerator2
    Surface Only=false
    UseOwnerScale=true
   ) {enabled=false}
   /SetVariables() {isSetParameters=true}
   &InitialMeshOrientation(
    Mesh Orientation Mode=::NewEnumerator2
    Rotation=&RandomRangeVector(Minimum=<0 0 -1>, Maximum=<0 0 1>)
    Use Rotation Vector=true
   )
  ]
  particleUpdateScript [
   &ParticleState(Kill Particles When Lifetime Has Elapsed=false, Let Infinitely Lived Particles Die When Emitter Deactivates=true)
   /Niagara/Noise/Waveform(
    EnableExponent1=false
    Global Input Position:NiagaraFloat=&RandomRangeFloat(Minimum=0, Maximum=1)
    Global Phase:NiagaraFloat=&Multiply_Float(A=$Emitter.Age, B=&RandomRangeFloat(Minimum=0.05, Maximum=0.075))
    GlobalClampMax=false
    GlobalClampMin=false
    UseInput1=false
    UsePhase1=false
    "[1] Amplitude Min/Max":Vector2f=<-1 1>
    "[1] Scale"=0.1
   )
   &Waveform(
    EnableExponent1=false
    Global Amplitude Scale=0.7
    Global Input Position:NiagaraFloat=&RandomRangeFloat(Minimum=0, Maximum=1)
    Global Phase:NiagaraFloat=&Multiply_Float(A=$Emitter.Age, B=&RandomRangeFloat(Minimum=0.05, Maximum=0.075))
    GlobalClampMax=false
    GlobalClampMin=false
    UseInput1=false
    UsePhase1=false
    "Waveform [1]":ENiagara_Waveforms=::NewEnumerator1
    "[1] Amplitude Min/Max":Vector2f=<-1 1>
    "[1] Scale"=0.1
   ) {module=Waveform001}
   &ScaleMeshSize(Scale Factor=&MakeVector(
     X=/Niagara/DynamicInputs/Helpers/Waveform(
      Global Input Position=&RandomRangeFloat(Minimum=0, Maximum=1)
      "Waveform [1]"=::NewEnumerator0
      "[1] Frequency"=1
      "[1] Amplitude Min/Max"=<0.7 1>
     )
     Y=1
     Z=1
    ))
   &ScaleSpriteSize(Scale Factor=&Multiply_Vector2D_ByFloat(
     Float=&Waveform(
      Global Input Position=&RandomRangeFloat(Minimum=0, Maximum=1)
      "Waveform [1]"=::NewEnumerator12
      "[1] Frequency"=5
      "[1] Amplitude Min/Max"=<1 2>
      "[1] Scale"=0.5
      "[1] Bias"=0.5
     )
     Vector2D=<1 1>
    )) {module=ScaleSpriteSize001}
   /SetVariables(Particles.MaterialRandom:NiagaraFloat=&RandomRangeFloat(Minimum=0, Maximum=1)) {isSetParameters=true}
   &SolveForcesAndVelocity(Clamp Velocity=false, Limit Acceleration=false)
   /SetVariables(Particles.Position=&AddVectorToPosition(
     Position=$Particles.Position
     Vector=&MakeVector(X=$Particles.Waveform001.Waveform, Y=0, Z=$Particles.Waveform.Waveform)
    )) {enabled=false, isSetParameters=true}
   /SetVariables(Particles.GoalPos:NiagaraPosition=&AddVectorToPosition(
     Position=$Particles.Initial.Position
     Vector=&Multiply_VectorByFloat(Vector=&MakeVector(X=$Particles.Waveform001.Waveform, Y=0, Z=$Particles.Waveform.Waveform), Float=150)
    )) {isSetParameters=true}
   &self:AvoidPlayer(PlayerPosOffset:Vector3f=<0 0 85>, PushFalloff=250, PushRadius=150) {module=AvoidPlayer}
   /SetVariables(Particles.Position=&AddVectorToPosition(Position=$Particles.GoalPos, Vector=$Output.AvoidPlayer.PushOffset)) {isSetParameters=true}
   &FN_Mesh_Attributes(Create Color Attribute=true, Create Scale Attribute=true, Scale=<2 2 2>, Write to Module Namespace=true)
   &DynamicMaterialParameters(
    Index 0 Param 1=&Lerp_Float(
     A=1
     B=2
     Alpha=&Waveform(
      Global Input Position=&RandomRangeFloat(Minimum=0, Maximum=1)
      "Waveform [1]"=::NewEnumerator12
      "[1] Frequency"=5
      "[1] Amplitude Min/Max"=<0 1>
     )
    )
    Module.Index0Param0Write=true
    Module.Index0Param1Write=false
    Module.Index0Param2Write=false
    Module.Index0Param3Write=false
   ) {module=DynamicMaterialParameters001}
   &FN_Sprite_Attributes(
    Color=<rgba 1 0.4366 0.2552 0.2>
    Create Color Attribute=true
    Create Position Attribute=true
    Position:NiagaraPosition=&AddVectorToPosition(Position=$Particles.Position, Vector=<0 0 2>)
    Write to Module Namespace=true
   )
  ]
  rendererClasses [&NiagaraMeshRendererProperties, &NiagaraSpriteRendererProperties]
  renderers [
   {rendererClass=&NiagaraMeshRendererProperties, rendererIndex=0}
   {rendererClass=&NiagaraSpriteRendererProperties, rendererIndex=1}
   {rendererClass=&NiagaraMeshRendererProperties, rendererIndex=2}
  ]
 }
 Snow {
  bEnabled=false
  simTarget=GPUComputeSim
  emitterUpdateScript [
   &EmitterState()
   &SpawnBurst_Instantaneous(Spawn Count=5000, Use Spawn Probability=false, UseLoopCountLimit=true)
   &SpawnRate(SpawnRate=5000, Use Spawn Probability=false)
  ]
  particleSpawnScript [
   /SetVariables(Particles.HorizontalDistribution:NiagaraFloat=&RandomRangeFloat(Minimum=0, Maximum=1)) {isSetParameters=true}
   &InitializeParticle(
    Color Mode=::NewEnumerator0
    Lifetime=&Multiply_Float(
     A=&RandomRangeFloat(Minimum=0.5, Maximum=1.5, Evaluation Type=::NewEnumerator1)
     B=&Lerp_Float(A=0.75, B=0.75, Alpha=$Particles.HorizontalDistribution)
    )
    Mass=&RandomRangeFloat(Minimum=0.3, Maximum=1)
    Mass Mode=::NewEnumerator1
    Sprite Rotation Angle=&RandomRangeFloat(Minimum=0, Maximum=1, Evaluation Type=::NewEnumerator1)
    Sprite Rotation Mode=::NewEnumerator1
    Sprite Size=&Vector2DFromFloat(Value=&RandomRangeFloat(Minimum=3, Maximum=9, Evaluation Type=::NewEnumerator1))
    Sprite Size Mode=::NewEnumerator4
    UsePositionOffset=false
    Write Material Random=true
   )
   /SetVariables(Particles.Position=&AddVectorToPosition(
     Position=&SimulationPosition()
     Vector=&MakeVector(
      X=&Lerp_Float(A=-750, B=750, Alpha=$Particles.HorizontalDistribution)
      Y=&RandomRangeFloat(Minimum=-800, Maximum=730)
      Z=&Lerp_Float(
       A=&RandomRangeFloat(Minimum=-40, Maximum=-50)
       B=&RandomRangeFloat(Minimum=-50, Maximum=300)
       Alpha=&FloatFromCurve(
        FloatCurve=DataInterface() {propertyValues {Curve {keys [
            <at 0 0>
            {arriveTangent=4.2262, interpMode=RCIM_Cubic, leaveTangent=3.5357, tangentMode=RCTM_Break, time=0.4, value=1}
            {interpMode=RCIM_Cubic, tangentMode=RCTM_User, time=0.5, value=1.25}
            {arriveTangent=-3.5052, tangentMode=RCTM_User, time=0.6, value=1}
            <at 1 0>
           ]}}}
        CurveIndex=$Particles.HorizontalDistribution
       )
      )
     )
    )) {isSetParameters=true}
   &AccelerationForce(Acceleration=&Multiply_VectorByFloat(Vector=<0 0 -40>, Float=&RandomRangeFloat(Minimum=0.2, Maximum=1)))
   &CurlNoiseForce(Bias Noise=false, Noise Frequency=25, Pan Noise=true, Pan Noise Field=<0 0 1>)
   &ApplyInitialForces(Apply Force to Position=true)
  ]
  particleUpdateScript [
   &ParticleState()
   &SpriteRotationRate(Rotation Rate=&RandomRangeFloat(Minimum=10, Maximum=100))
   &ScaleSpriteSize(
    Scale Factor=&Vector2DFromFloat(Value=&FloatFromCurve(
      FloatCurve=DataInterface() {propertyValues {Curve {
         defaultValue=0
         keys [
          {interpMode=RCIM_Cubic, tangentMode=RCTM_User, time=0, value=0}
          {arriveTangent=1.3914, interpMode=RCIM_Cubic, leaveTangent=1.3914, tangentMode=RCTM_User, time=0.25, value=1}
          {arriveTangent=-1.3914, interpMode=RCIM_Cubic, leaveTangent=-1.3914, tangentMode=RCTM_User, time=0.75, value=1}
          {tangentMode=RCTM_User, time=1, value=0}
         ]
        }}}
      Scale Curve=0.8
     ))
    Scale Sprite Size Mode=::NewEnumerator2
   ) {module=ScaleSpriteSize001}
   &ScaleColor(
    Scale Alpha=&FloatFromCurve(FloatCurve=DataInterface() {propertyValues {Curve {
        defaultValue=1
        keys [
         {interpMode=RCIM_Cubic, tangentMode=RCTM_User, time=0, value=0}
         {interpMode=RCIM_Cubic, tangentMode=RCTM_User, time=0.5, value=1}
         {tangentMode=RCTM_User, time=1, value=0}
        ]
       }}})
    ScaleA=true
    ScaleRGB=true
   )
   &CurlNoiseForce(Bias Noise=false, Pan Noise=false) {module=CurlNoiseForce001}
   &SolveForcesAndVelocity(Clamp Velocity=false, Limit Acceleration=false)
   /Niagara/Modules/Update/Camera/MaintainInCameraParticleScale(Render Target Width=3)
  ]
  rendererClasses [&NiagaraSpriteRendererProperties]
  renderers [{rendererClass=&NiagaraSpriteRendererProperties}]
 }
 Snow_Low_End {
  bEnabled=false
  emitterUpdateScript [
   &EmitterState()
   &SpawnRate(SpawnRate=120, Use Spawn Probability=false)
  ]
  particleSpawnScript [
   /SetVariables(Particles.HorizontalDistribution:NiagaraFloat=&RandomRangeFloat(Minimum=0, Maximum=1)) {isSetParameters=true}
   &InitializeParticle(
    Color Mode=::NewEnumerator0
    Lifetime=&Multiply_Float(
     A=&RandomRangeFloat(Minimum=0.5, Maximum=1.5, Evaluation Type=::NewEnumerator1)
     B=&Lerp_Float(A=0.5, B=1, Alpha=$Particles.HorizontalDistribution)
    )
    Mass=&RandomRangeFloat(Minimum=0.3, Maximum=1)
    Mass Mode=::NewEnumerator1
    Sprite Rotation Angle=&RandomRangeFloat(Minimum=0, Maximum=1, Evaluation Type=::NewEnumerator1)
    Sprite Rotation Mode=::NewEnumerator1
    Sprite Size=&Vector2DFromFloat(Value=&RandomRangeFloat(Minimum=3, Maximum=9, Evaluation Type=::NewEnumerator1))
    Sprite Size Mode=::NewEnumerator4
    UsePositionOffset=false
    Write Material Random=true
   )
   /SetVariables(Particles.Position=&AddVectorToPosition(
     Position=&SimulationPosition()
     Vector=&MakeVector(
      X=&Lerp_Float(A=-750, B=750, Alpha=$Particles.HorizontalDistribution)
      Y=&RandomRangeFloat(Minimum=-800, Maximum=730)
      Z=&Lerp_Float(
       A=&RandomRangeFloat(Minimum=-40, Maximum=-50)
       B=&RandomRangeFloat(Minimum=-50, Maximum=300)
       Alpha=&FloatFromCurve(
        FloatCurve=DataInterface() {propertyValues {Curve {keys [
            <at 0 0>
            {arriveTangent=4.2262, interpMode=RCIM_Cubic, leaveTangent=3.5357, tangentMode=RCTM_Break, time=0.4, value=1}
            {interpMode=RCIM_Cubic, tangentMode=RCTM_User, time=0.5, value=1.25}
            {arriveTangent=-3.5052, tangentMode=RCTM_User, time=0.6, value=1}
            <at 1 0>
           ]}}}
        CurveIndex=$Particles.HorizontalDistribution
       )
      )
     )
    )) {isSetParameters=true}
   &AccelerationForce(Acceleration=&Multiply_VectorByFloat(Vector=<0 0 -40>, Float=&RandomRangeFloat(Minimum=0.2, Maximum=1)))
   &CurlNoiseForce(Bias Noise=false, Noise Frequency=25, Pan Noise=true, Pan Noise Field=<0 0 1>)
   &ApplyInitialForces(Apply Force to Position=true)
  ]
  particleUpdateScript [
   &ParticleState()
   &SpriteRotationRate(Rotation Rate=&RandomRangeFloat(Minimum=10, Maximum=100))
   &ScaleSpriteSize(
    Scale Factor=&Vector2DFromFloat(Value=&FloatFromCurve(
      FloatCurve=DataInterface() {propertyValues {Curve {
         defaultValue=0
         keys [
          {interpMode=RCIM_Cubic, tangentMode=RCTM_User, time=0, value=0}
          {arriveTangent=1.3914, interpMode=RCIM_Cubic, leaveTangent=1.3914, tangentMode=RCTM_User, time=0.25, value=1}
          {arriveTangent=-1.3914, interpMode=RCIM_Cubic, leaveTangent=-1.3914, tangentMode=RCTM_User, time=0.75, value=1}
          {tangentMode=RCTM_User, time=1, value=0}
         ]
        }}}
      Scale Curve=0.8
     ))
    Scale Sprite Size Mode=::NewEnumerator2
   ) {module=ScaleSpriteSize001}
   &ScaleColor(
    Scale Alpha=&FloatFromCurve(FloatCurve=DataInterface() {propertyValues {Curve {
        defaultValue=1
        keys [
         {interpMode=RCIM_Cubic, tangentMode=RCTM_User, time=0, value=0}
         {interpMode=RCIM_Cubic, tangentMode=RCTM_User, time=0.5, value=1}
         {tangentMode=RCTM_User, time=1, value=0}
        ]
       }}})
    ScaleA=true
    ScaleRGB=true
   )
   &CurlNoiseForce(Bias Noise=false, Pan Noise=false) {module=CurlNoiseForce001}
   &SolveForcesAndVelocity(Clamp Velocity=false, Limit Acceleration=false)
   &MaintainInCameraParticleScale(Render Target Width=3)
  ]
  rendererClasses [&NiagaraSpriteRendererProperties]
  renderers [{rendererClass=&NiagaraSpriteRendererProperties}]
 }
}
summary {
 systemName=NS_S3911_WinterfestCabin
 emitters [
  {
   emitterName=Foreground_Fog001
   rendererClasses [&NiagaraMeshRendererProperties, &NiagaraSpriteRendererProperties]
  }
  {bEnabled=false, emitterName=Snow, simTarget=GPUComputeSim, rendererClasses [&NiagaraSpriteRendererProperties]}
  {bEnabled=false, emitterName=Snow_Low_End, rendererClasses [&NiagaraSpriteRendererProperties]}
 ]
 userVariables []
}
